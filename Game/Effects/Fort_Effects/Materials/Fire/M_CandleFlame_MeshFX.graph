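# Translucent unlit material for a candle flame on Niagara mesh particles, with a Candle texture driving emissive and depth-faded opacity.
# Panning noise, pulsed by a randomly offset sine, distorts the flame UVs, and an optional two-color vertical gradient tints it.
# World position offset combines SplineThicken width, particle-driven drag weighted along the UV, an optional pivot bend and camera offset.
# Object position and a particle parameter supply randomization.
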Material /Game/Effects/Fort_Effects/Materials/Fire/M_CandleFlame_MeshFX
graph {
 properties {
  bEnableResponsiveAA=true
  bUsedWithNiagaraMeshParticles=true
  blendMode=BLEND_Translucent
  pixelDepthOffsetMode=PDOM_Legacy
  shadingModel=MSM_Unlit
  usedShadingModels=Unlit
 }
 nodes [
  n1 = DynamicParameter() {paramNames [SplineThickness, DragX, DragY, DragZ]}
  n2 = ScalarParameter() {group=SplineThicken, parameterName=Width, sliderMax=50}
  n3 = Constant() {r=1.25}
  n4 = Add(A=@n1.SplineThickness, B=LinearInterpolate(A=@n2, B=Multiply(A=@n2, B=@n3), Alpha=NamedRerouteUsage()))
  n5 = OneMinus(None=ComponentMask(None=TextureCoordinate()) {g=true})
  n6 = AppendVector(A=AppendVector(A=@n1.DragX, B=@n1.DragX), B=@n1.DragZ)
  n7 = Constant3Vector() {constant=<rgba -10 -10 0 0>}
  n8 = Constant3Vector() {constant=<rgba 10 10 10 0>}
  n9 = Add(
   A=MaterialFunctionCall(WidthBase=@n4, WidthTip=@n4) {materialFunction=&SplineThicken}.WorldPositionOffset
   B=Multiply(A=Power(Base=@n5), B=Clamp(None=@n6, Min=@n7, Max=@n8) {maxDefault=20, minDefault=-20})
  )
  n10 = VectorParameter() {group=Bend, parameterName=BendAxis, defaultValue=<rgba 0 0 0 1>}
  n11 = Power(
   Base=OneMinus(None=MaterialFunctionCall() {materialFunction=&BoundingBoxBased_0-1_UVW}.R)
   Exp=ScalarParameter() {defaultValue=1, group=Bend, parameterName=BendPower}
  )
  n12 = RotateAboutAxis(
   NormalizedRotationAxis=Transform(None=LinearInterpolate(B=@n10.RGB, Alpha=@n11)) {transformSourceType=TRANSFORMSOURCE_Local}
   RotationAngle=ScalarParameter() {group=Bend, parameterName=BendAngle}
   PivotPoint=MaterialFunctionCall() {materialFunction=&ComponentPivotLocation}.Result
   Position=WorldPosition().XYZ
  )
  Add(
   A=LinearInterpolate(A=@n9, B=Add(A=@n9, B=@n12), Alpha=ScalarParameter() {group=Bend, parameterName=BendSpline?})
   B=MaterialFunctionCall(Offset Amount=ScalarParameter() {parameterName=CameraOffset}) {materialFunction=/Engine/Functions/Engine_MaterialFunctions02/WorldPositionOffset/CameraOffset}.Camera Offset
  )
  Constant()
  DynamicParameter() {parameterIndex=1, paramNames [Randomize,Param2,Param3,Param4]}
  n13 = MaterialFunctionCall(Float2=TextureCoordinate()) {materialFunction=&BreakOutFloat2Components}
  n14 = AppendVector(A=Add(A=NamedRerouteUsage(), B=@n13.R), B=@n13.G)
  n15 = VectorParameter() {parameterName=DistortionTiling, defaultValue=<rgba 1 1 1 1>}
  n16 = ComponentMask(None=@n15.RGB) {g=true, r=true}
  n17 = Time()
  n18 = NamedRerouteUsage()
  n19 = ScalarParameter() {defaultValue=1, parameterName=DistortionTilingMult}
  n20 = Multiply(
   A=Panner(
    Coordinate=Multiply(A=@n14, B=@n16)
    Time=Add(A=@n17, B=@n18)
    Speed=MaterialFunctionCall(R=@n15.B, G=@n15.B) {materialFunction=&AppendMany}.RG
   ) {bFractionalPart=true, speedX=0.5}
   B=LinearInterpolate(A=@n19, B=Multiply(A=@n19) {constB=1.25}, Alpha=NamedRerouteUsage())
  )
  n21 = Subtract(
   A=Multiply(
    A=TextureSampleParameter2D(UVs=@n20) {parameterName=DistortionTexture, texture=/Game/Effects/Fort_Effects/Textures/Tile/T_GaussianNoise_01_M}.G
    B=Constant() {r=2}
   )
   B=ScalarParameter() {defaultValue=1, parameterName=Bias}
  ) {constB=1.5}
  n22 = Multiply(
   A=ComponentMask(None=VectorParameter() {parameterName=DistStr, defaultValue=<rgba 0.2 0.1 0 1>}.RGB) {g=true, r=true}
   B=ScalarParameter() {defaultValue=1, parameterName=Distortionstrength}
  )
  n23 = NamedRerouteUsage()
  n24 = NamedRerouteUsage()
  n25 = Add(A=Time(), B=Add(A=Add(A=@n23, B=@n24), B=PerInstanceRandom()))
  n26 = Add(
   A=OneMinus(None=MaterialFunctionCall() {materialFunction=&Swizzle}.YX)
   B=Multiply(A=Multiply(A=@n21, B=@n22), B=Sine(None=@n25))
  )
  n27 = TextureSample(UVs=@n26) {samplerSource=SSM_Clamp_WorldGroupSettings, samplerType=SAMPLERTYPE_LinearColor, texture=/Game/Effects/Fort_Effects/Textures/Fire/Candle}
  n28 = Multiply(A=ScalarParameter() {defaultValue=1, parameterName=Emissive}, B=@n27.G)
  n29 = TextureCoordinate()
  n30 = LinearInterpolate(
   A=VectorParameter() {parameterName=Color1, defaultValue=<rgba 2 1.376 0.3148 1>}.RGB
   B=VectorParameter() {parameterName=Color2, defaultValue=<rgba 0 0.6896 9 1>}.RGB
   Alpha=SmoothStep(
    Min=ScalarParameter() {defaultValue=0.3, parameterName=GradientSmoothStep_Min}
    Max=ScalarParameter() {defaultValue=0.8, parameterName=GradientSmoothStep_Max}
    Value=ComponentMask(None=@n29) {g=true}
   )
  )
  MaterialFunctionCall(Emissive=LinearInterpolate(A=@n28, B=Multiply(A=@n30, B=@n28), Alpha=ScalarParameter() {parameterName=UseColorGradiant?})) {materialFunction=&MF_FX_EmissiveAdjustments}
  n31 = ScalarParameter() {parameterName=DepthFadeDistance}
  Multiply(A=Multiply(A=@n27.G, B=DepthFade(FadeDistance=@n31)), B=ScalarParameter() {defaultValue=1, parameterName=Opacity})
  n32 = Multiply(A=ObjectPositionWS(), B=ScalarParameter() {defaultValue=1, parameterName=MultObjectPosition})
  NamedRerouteDeclaration(None=ComponentMask(None=Frac(None=@n32)) {g=true}) {name=PerComponentRandom, nodeColor=<rgba 1 0 0.4219 1>}
  NamedRerouteDeclaration(None=DynamicParameter() {parameterIndex=1, paramNames [Randomize,Param2,Param3,Param4]}.Randomize) {name=ParticleRandom, nodeColor=<rgba 0 0.25 1 1>}
  ParticleColor()
  ScalarParameter() {parameterName=SplineThickness}
  TextureSampleParameter2D(UVs=@n26) {parameterName=FlameTexture, samplerType=SAMPLERTYPE_LinearColor, texture=&Candle}
 ]
}
