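# Lit material for cozy campfire wood that chars and glows with embers as the Burn Amount parameter rises.
# An actor-relative height mask, shifted by Burn Amount and broken up by char textures, blends woodpile color and normals into charred ones.
# Two layers of scrolling world-space noise flicker the hot areas, which index a hot-coals curve atlas for HDR emissive.
# A small vertex-normal offset is also computed.
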Material /Game/Effects/Fort_Effects/Materials/Fire/M_CozyCampfire_Embers
graph {
 properties {bUsedWithInstancedStaticMeshes=true, bUsedWithNanite=true, pixelDepthOffsetMode=PDOM_Legacy, usedShadingModels=Default Lit}
 nodes [
  n1 = ScalarParameter() {defaultValue=0.2, parameterName=Power}
  n2 = TextureCoordinate() {uTiling=2.5, vTiling=2.5}
  n3 = TextureSample(UVs=@n2) {
   samplerSource=SSM_Wrap_WorldGroupSettings
   texture=/Game/Environments/Apollo/Sets/CoralBuddies/Props/Textures/T_CoralBuddies_AfterCredits_RubbleChar__01_D
  }
  n4 = TextureSample(UVs=TextureCoordinate() {vTiling=2.5}) {
   samplerSource=SSM_Wrap_WorldGroupSettings
   texture=&T_CoralBuddies_AfterCredits_RubbleChar__01_D
  }
  n5 = TextureSample() {samplerSource=SSM_Wrap_WorldGroupSettings, texture=/Game/Environments/Props/WildWest/Textures/WildWest_WoodPile_1_D}
  n6 = Multiply(A=@n5.B) {constB=50}
  n7 = Multiply(A=Multiply(A=@n6, B=@n6), B=@n5.B)
  n8 = LinearInterpolate(A=@n3.RGB, B=Multiply(A=@n4.RGB) {constB=0.23}, Alpha=Clamp(None=Multiply(A=@n7, B=@n5.B))) {constA=1, constB=0.25}
  n9 = Power(Base=@n8, Exp=@n1) {constExponent=1}
  n10 = WorldPosition()
  n11 = ActorPositionWS()
  n12 = ScalarParameter() {defaultValue=-20, parameterName=Mask Vertical Offset Min}
  n13 = ScalarParameter() {defaultValue=-60, parameterName=Mask Vertical Offset Max}
  n14 = ScalarParameter() {parameterName=Burn Amount}
  n15 = ComponentMask(None=Add(A=Subtract(A=@n10.XYZ, B=@n11), B=LinearInterpolate(A=@n12, B=@n13, Alpha=@n14))) {b=true}
  n16 = OneMinus(None=Clamp(None=Multiply(A=@n15) {constB=0.025}))
  n17 = Power(Base=Clamp(None=Subtract(A=@n16, B=@n9))) {constExponent=5}
  n18 = TextureCoordinate()
  n19 = Subtract(
   A=@n16
   B=Power(
    Base=TextureSample(UVs=@n18) {texture=&T_CoralBuddies_AfterCredits_RubbleChar__01_D}.G
    Exp=@n1
   ) {constExponent=1}
  )
  n20 = Multiply(A=@n19) {constB=2}
  n21 = WorldPosition()
  n22 = Time()
  n23 = AppendVector(A=Multiply(A=@n22) {constB=-10}, B=Multiply(A=@n22) {constB=8})
  n24 = Multiply(A=@n22) {constB=-16}
  n25 = Multiply(A=Add(A=@n21.XYZ, B=AppendVector(A=@n23, B=@n24))) {constB=0.015}
  n26 = TextureSample(UVs=ComponentMask(None=@n25) {b=true, r=true}) {texture=/Game/Global/Textures/Default/Noise/T_FireNoise_02}
  n27 = TextureSample(UVs=ComponentMask(None=@n25) {b=true, g=true}) {texture=&T_FireNoise_02}
  n28 = Power(Base=Divide(A=Add(A=@n26.G, B=@n27.G)))
  n29 = Time()
  n30 = Multiply(A=@n29) {constB=-24}
  n31 = AppendVector(A=@n30, B=@n30)
  n32 = Multiply(A=@n29) {constB=-32}
  n33 = Multiply(A=Add(A=@n21.XYZ, B=AppendVector(A=@n31, B=@n32))) {constB=0.0066}
  n34 = TextureSample(UVs=ComponentMask(None=@n33) {b=true, r=true}) {
   samplerType=SAMPLERTYPE_LinearColor
   texture=/Engine/Functions/Engine_MaterialFunctions02/ExampleContent/Textures/OffsetNoiseDistanceFields_ForNormals
  }
  n35 = TextureSample(UVs=ComponentMask(None=@n33) {b=true, g=true}) {
   samplerType=SAMPLERTYPE_LinearColor
   texture=&OffsetNoiseDistanceFields_ForNormals
  }
  n36 = Power(Base=Divide(A=Add(A=@n34.G, B=@n35.G))) {constExponent=4}
  n37 = Multiply(A=Multiply(A=Saturate(None=@n20), B=Multiply(A=@n28, B=@n36))) {constB=15}
  n38 = Multiply(A=Saturate(None=Multiply(A=@n17) {constB=4}), B=Saturate(None=Add(A=@n37) {constB=0.04}))
  n39 = CurveAtlasRowParameter(CurveTime=ComponentMask(None=Add(A=@n17, B=@n38)) {g=true}) {
   atlas=/Game/Effects/Niagara/MasterMaterials/Smoke/Textures/CurveAtlas
   curve=/Game/Effects/Niagara/MasterMaterials/Smoke/Textures/EmissiveCurveHotCoals
   defaultValue=5
  }
  n40 = Multiply(A=@n39, B=ScalarParameter() {parameterName=Emissive})
  MaterialFunctionCall(LDR Input=@n40, HDR Tint=@n40, HDR Intensity Multiplier=ScalarParameter() {defaultValue=3, parameterName=HDR Intensity}) {materialFunction=&DeriveHDRfromLDR}
  Reroute(None=Multiply(A=VertexNormalWS(), B=ScalarParameter() {defaultValue=0.01, parameterName=Offset Amount}) {constB=1.005})
  n41 = Saturate(None=Multiply(A=@n19) {constB=7})
  Reroute(None=LinearInterpolate(A=@n5.RGB, B=@n8, Alpha=@n41))
  n42 = TextureSample() {
   samplerSource=SSM_Wrap_WorldGroupSettings
   samplerType=SAMPLERTYPE_Normal
   texture=/Game/Environments/Props/WildWest/Textures/WildWest_WoodPile_1_N
  }
  n43 = TextureSample() {samplerSource=SSM_Wrap_WorldGroupSettings, samplerType=SAMPLERTYPE_Normal, texture=/Engine/EngineMaterials/FlatNormal}
  n44 = TextureSample(UVs=@n2) {
   samplerSource=SSM_Wrap_WorldGroupSettings
   samplerType=SAMPLERTYPE_Normal
   texture=/Game/Environments/Apollo/Sets/CoralBuddies/Props/Textures/T_CoralBuddies_AfterCredits_RubbleChar__01_N
  }
  Reroute(None=MaterialFunctionCall(BaseNormal=@n42.RGB, AdditionalNormal=LinearInterpolate(A=@n43.RGB, B=@n44.RGB, Alpha=@n41)) {materialFunction=&BlendAngleCorrectedNormals}.Result)
  TextureSample() {
   samplerSource=SSM_Wrap_WorldGroupSettings
   samplerType=SAMPLERTYPE_Masks
   texture=/Game/Environments/Props/WildWest/Textures/WildWest_WoodPile_1_S
  }
 ]
}
