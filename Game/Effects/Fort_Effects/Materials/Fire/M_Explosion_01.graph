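# Translucent, volumetric-lit explosion particle material that builds glowing fire and smoke from SubUV flipbook channels, usable on mesh particles.
# A radial core mask, offset and scaled noise detail and the flipbook's blue edge channel drive a three-colour particle-colour blend for emissive.
# Dynamic parameters control detail and falloff per particle.
# Opacity sums core, edge and green channel, with depth fade and HDR remap off mobile, plus switchable camera and distance fades.
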Material /Game/Effects/Fort_Effects/Materials/Fire/M_Explosion_01
graph {
 properties {
  bUsedWithMeshParticles=true
  blendMode=BLEND_Translucent
  pixelDepthOffsetMode=PDOM_Legacy
  translucencyLightingMode=TLM_VolumetricPerVertexDirectional
  usedShadingModels=Default Lit
 }
 nodes [
  n1 = ParticleColor()
  n2 = Power(
   Base=TextureSampleParameter2D(UVs=TextureCoordinate()) {parameterName=ExplosionEmissiveGlowMask, texture=/Game/UI/Frontend/Store/Textures/T_RadGradient}.A
   Exp=DynamicParameter() {paramNames [DetailOffset, DetailScale, FireCorePower, EdgePower]}.FireCorePower
  )
  n3 = TextureCoordinate() {uTiling=2, vTiling=2}
  n4 = DynamicParameter() {paramNames [DetailOffset, DetailScale, FireCorePower, EdgePower]}
  n5 = AppendVector(A=@n4.DetailOffset, B=Add(A=@n4.DetailOffset, B=Constant() {r=1}))
  n6 = TextureSampleParameter2D(UVs=Multiply(A=Add(A=@n3, B=@n5), B=Floor(None=@n4.DetailScale)) {desc=Scale}) {parameterName=ExplosionEmissiveDetailTexture, texture=/Game/Effects/Fort_Effects/Textures/Fire/T_FireNoise_02}
  n7 = TextureSampleParameterSubUV!(UVs, Apply View MipBias) {parameterName=AlphaG_EmissiveB, samplerType=SAMPLERTYPE_LinearColor, texture=/Game/Effects/Fort_Effects/Textures/Smoke/SmokeTest2_Soft}
  n8 = Power(Base=@n7.B, Exp=DynamicParameter() {paramNames [DetailOffset, DetailScale, FireCorePower, EdgePower]}.EdgePower)
  n9 = MaterialFunctionCall(A=Constant3Vector(), B=Multiply(A=@n1.RGB) {constB=0.5}, C=@n1.RGB, Alpha=Multiply(A=Multiply(A=@n2, B=@n6.G), B=@n8)) {materialFunction=&3ColorBlend}
  n10 = Clamp(None=@n9.Result)
  MaterialFunctionCall(Emissive=ShadingPathSwitch(
    Default=MaterialFunctionCall(LDR Input=@n10, HDR Tint=@n9.Result) {materialFunction=&DeriveHDRfromLDR}.RemappedRange
    Mobile=@n9.Result
   ) {desc=Migrated from FeatureLevelSwitch}) {materialFunction=&MF_FX_EmissiveAdjustments}
  n11 = Add(A=@n2, B=@n8)
  n12 = Multiply(A=Clamp(None=Add(A=@n11, B=@n7.G)), B=ParticleColor().A)
  n13 = ShadingPathSwitch(
   Default=DepthFade(Opacity=@n12, FadeDistance=ScalarParameter() {defaultValue=140, parameterName=DepthFadeDistance}) {fadeDistanceDefault=140}
   Mobile=@n12
  ) {desc=Migrated from FeatureLevelSwitch}
  n14 = ScalarParameter() {defaultValue=150, parameterName=Near Camera Fade Distance}
  n15 = StaticSwitchParameter(
   True=Multiply(A=@n13, B=MaterialFunctionCall(Distance=@n14) {materialFunction=&Near_Camera_Fade}.Result)
   False=@n13
  ) {parameterName=Near Camera Fade}
  StaticSwitchParameter(True=Multiply(A=DistanceCullFade(), B=@n15), False=@n15) {parameterName=UseDistanceCullFade}
  TextureSampleParameterSubUV() {parameterName=Normal, samplerType=SAMPLERTYPE_Normal, texture=/Game/Effects/Fort_Effects/Textures/Smoke/SmokeTest_N}
  VectorParameter() {parameterName=BaseColor, defaultValue=<rgba 0.2997 0.5125 0.6042 0>}
 ]
}
