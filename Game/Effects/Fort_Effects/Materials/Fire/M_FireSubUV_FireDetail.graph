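# Unlit translucent particle sprite material for SubUV fire with added noise detail.
# Flipbook alpha is blended toward a noise-eroded copy, with dynamic parameters controlling blend, UV offset, erosion power and floored tiling scale.
# Particle color multiplies that mask, feeding emissive through HDR-remap and adjustment functions.
# Opacity gets depth fade with mask-biased distance, skipped on mobile, plus an optional static-switched near-camera fade.
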Material /Game/Effects/Fort_Effects/Materials/Fire/M_FireSubUV_FireDetail
graph {
 properties {
  bUsedWithParticleSprites=true
  blendMode=BLEND_Translucent
  pixelDepthOffsetMode=PDOM_Legacy
  shadingModel=MSM_Unlit
  usedShadingModels=Unlit
 }
 nodes [
  n1 = ParticleColor()
  n2 = TextureSampleParameterSubUV!(UVs, Apply View MipBias) {parameterName=FireSubUV, texture=/Game/Effects/Fort_Effects/Textures/Fire/T_FireFluid_1_8X8}
  n3 = TextureCoordinate() {uTiling=8, vTiling=8}
  n4 = DynamicParameter() {defaultValue=<rgba 0 0 0 1>, paramNames [LerpNoise, NoiseOffsetX, NoiseErosionPower, NoiseScale]}
  n5 = TextureSampleParameter2D(UVs=Multiply(A=Add(A=@n3, B=@n4.NoiseOffsetX), B=Floor(None=@n4.NoiseScale)) {desc=Scale}) {parameterName=DetailTexture, samplerType=SAMPLERTYPE_LinearColor, texture=/Game/Global/Textures/Default/Noise/T_Noise01}
  n6 = LinearInterpolate(
   A=@n2.A
   B=Multiply(A=@n2.A, B=Power(Base=@n5.G, Exp=@n4.NoiseErosionPower))
   Alpha=DynamicParameter() {defaultValue=<rgba 0 0 0 1>, paramNames [LerpNoise, NoiseOffsetX, NoiseErosionPower, NoiseScale]}.LerpNoise
  )
  n7 = Multiply(A=@n1.RGB, B=@n6)
  MaterialFunctionCall(Emissive=MaterialFunctionCall(LDR Input=Clamp(None=@n7), HDR Tint=@n7) {materialFunction=&DeriveHDRfromLDR}.RemappedRange) {materialFunction=&MF_FX_EmissiveAdjustments}
  n8 = Multiply(A=@n1.A, B=@n6)
  n9 = ShadingPathSwitch(Default=DepthFade(Opacity=@n8, FadeDistance=LinearInterpolate(Alpha=@n6) {constA=60, constB=120, desc=Bias Noise}), Mobile=@n8) {desc="Added pref for mobile,Yang."}
  n10 = ScalarParameter() {defaultValue=150, parameterName=Near Camera Distance}
  StaticSwitchParameter(
   True=Multiply(A=@n9, B=MaterialFunctionCall(Distance=@n10) {materialFunction=&Near_Camera_Fade}.Result)
   False=@n9
  ) {parameterName=Use Near Camera Fade}
 ]
}
