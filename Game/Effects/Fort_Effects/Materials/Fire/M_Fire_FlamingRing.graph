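# Translucent, unlit, two-sided mesh-particle material for a flaming ring effect, with dynamic-parameter UV offsets.
# Three panned noise layers plus a vertical gradient drive a two-color HDR blend, tinted by particle color and optionally noise hue-shifted.
# Vertex displacement combines normal-directed noise, a scrolling gradient and animated vector noise, fading along V.
# Opacity combines flame, gradient masks and optional alpha textures; mobile/forward paths skip depth fade, vector noise and gradient hue shift.
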Material /Game/Effects/Fort_Effects/Materials/Fire/M_Fire_FlamingRing
graph {
 properties {
  bUsedWithMeshParticles=true
  bUsedWithNiagaraMeshParticles=true
  blendMode=BLEND_Translucent
  pixelDepthOffsetMode=PDOM_Legacy
  shadingModel=MSM_Unlit
  twoSided=true
  usedShadingModels=Unlit
 }
 nodes [
  n1 = TextureCoordinate() {vTiling=0.05}
  n2 = DynamicParameter() {paramNames [UVOffset,RandomU,RandomV,Param4]}
  n3 = Add(A=ComponentMask(None=@n1) {r=true}, B=@n2.RandomU)
  n4 = Add(A=ComponentMask(None=@n1) {g=true}, B=@n2.RandomV)
  n5 = Constant()
  n6 = ScalarParameter() {defaultValue=0.1, parameterName=Hueshift Panning Speed}
  n7 = TextureSampleParameter2D(UVs=Panner(Coordinate=AppendVector(A=@n3, B=@n4), Speed=AppendVector(A=@n5, B=@n6)) {speedY=1}) {parameterName=HueshiftNoise, samplerType=SAMPLERTYPE_LinearColor, texture=/Game/Global/Textures/Default/Noise/T_Noise01_Inverse}
  n8 = Clamp(None=Power(Base=@n7.R, Exp=ScalarParameter() {defaultValue=3, parameterName=Hueshift Noise Exponent}))
  n9 = Power(
   Base=MaterialFunctionCall() {materialFunction=&LinearGradient}.VGradient
   Exp=ScalarParameter() {defaultValue=1, parameterName=HueShift_Line}
  )
  n10 = ScalarParameter() {parameterName=HueShifit_Amount}
  n11 = ComponentMask(None=VectorParameter() {group=UVScale, parameterName=UVScale, defaultValue=<rgba 1 1 0 0>}.RGB) {g=true, r=true}
  n12 = Multiply(A=TextureCoordinate() {vTiling=0.5}, B=@n11)
  n13 = DynamicParameter() {paramNames [UVOffset,RandomU,RandomV,Param4]}
  n14 = Multiply(A=Time(), B=ScalarParameter() {defaultValue=4, parameterName=Speed})
  n15 = TextureSample(UVs=Panner(Coordinate=Add(A=@n12, B=@n13.UVOffset), Time=@n14) {speedX=-0.05, speedY=1}) {texture=&T_NoisePack_01}
  n16 = Constant() {r=0.2}
  n17 = TextureCoordinate() {uTiling=4}
  n18 = Panner(Coordinate=Add(A=Multiply(A=@n11, B=@n17), B=@n13.UVOffset), Time=@n14) {speedX=0.1, speedY=1}
  n19 = Multiply(A=@n11, B=TextureCoordinate() {uTiling=12, vTiling=3})
  n20 = Multiply(
   A=Add(
    A=MaterialFunctionCall(In=@n15.G, Contrast=@n16) {materialFunction=&CheapContrast}.Result
    B=TextureSample(UVs=@n18) {texture=&T_NoisePack_01}.G
   )
   B=TextureSample(UVs=Panner(Coordinate=@n19, Time=@n14) {speedX=-0.1, speedY=1}) {texture=&T_NoisePack_01}.G
  )
  n21 = Add(
   A=@n20
   B=Power(
    Base=MaterialFunctionCall() {materialFunction=&LinearGradient}.VGradient
    Exp=ScalarParameter() {defaultValue=15, parameterName=FireLine}
   )
  )
  n22 = Clamp(None=MaterialFunctionCall(In=Power(Base=@n21) {constExponent=0.5}, Contrast=ScalarParameter() {defaultValue=0.5, parameterName=HardEdgeness}) {materialFunction=&CheapContrast}.Result)
  n23 = LinearInterpolate(
   A=VectorParameter() {parameterName=Color_A, defaultValue=<rgba 10 0 0.1768 1>}.RGB
   B=VectorParameter() {parameterName=Color_B, defaultValue=<rgba 10 1.3123 0 1>}.RGB
   Alpha=@n22
  )
  n24 = Clamp(None=@n23)
  n25 = ParticleColor()
  n26 = Multiply(A=@n23, B=@n25.RGB)
  n27 = Multiply(A=@n26, B=ScalarParameter() {defaultValue=1, parameterName=MobileColorMult})
  n28 = ShadingPathSwitch(
   Default=MaterialFunctionCall(
    Hue Shift Percentage=Multiply(A=@n9, B=@n10)
    Texture=MaterialFunctionCall(LDR Input=@n24, HDR Tint=@n26) {materialFunction=&DeriveHDRfromLDR}.RemappedRange
   ) {materialFunction=&HueShift}.Result
   Forward=@n27
   Mobile=@n27
  )
  MaterialFunctionCall(Emissive=StaticSwitchParameter(
    True=MaterialFunctionCall(Hue Shift Percentage=@n8, Texture=@n28) {materialFunction=&HueShift}.Result
    False=@n28
   ) {parameterName=Use Hueshift Noise}) {materialFunction=&MF_FX_EmissiveAdjustments}
  n29 = ScalarParameter() {defaultValue=25, parameterName=Distort}
  n30 = Multiply(A=Multiply(A=VertexNormalWS(), B=@n20), B=@n29)
  n31 = Add(A=@n13.UVOffset, B=TextureCoordinate())
  n32 = Multiply(
   A=TextureSample(UVs=Panner(Coordinate=@n31, Time=@n14) {speedY=0.25}) {texture=/Game/Global/Textures/Default/Masks_and_Gradients/CylGradient}.RGB
   B=@n29
  ) {constB=35}
  n33 = Subtract(
   A=Subtract(A=WorldPosition().XYZ, B=ObjectPositionWS())
   B=Multiply(A=ScalarParameter() {defaultValue=300, parameterName=Curl_Speed}, B=Time())
  )
  n34 = ScalarParameter() {defaultValue=350, parameterName=Curl_Scale}
  n35 = Multiply(
   A=ScalarParameter() {defaultValue=100, parameterName=Curl_Intensity}
   B=VectorNoise(World Position=Divide(A=@n33, B=@n34)) {noiseFunction=VNF_VectorALU}
  )
  n36 = Constant()
  n37 = MaterialFunctionCall() {materialFunction=&LinearGradient}
  n38 = ScalarParameter() {defaultValue=1, parameterName=FireLine_1}
  Multiply(
   A=Add(A=Add(A=@n30, B=@n32), B=ShadingPathSwitch(Default=@n35, Forward=@n36, Mobile=@n36))
   B=OneMinus(None=Power(Base=@n37.VGradient, Exp=@n38))
  )
  n39 = Multiply(
   A=@n22
   B=Power(
    Base=MaterialFunctionCall() {materialFunction=&LinearGradient}.VGradient
    Exp=ScalarParameter() {defaultValue=8, parameterName=GradientContrast}
   )
  )
  n40 = ShadingPathSwitch(
   Default=DepthFade(Opacity=@n39, FadeDistance=ScalarParameter() {defaultValue=64, parameterName=FadeDistance}) {fadeDistanceDefault=50}
   Forward=@n39
   Mobile=@n39
  )
  n41 = Constant() {r=0.025}
  n42 = TextureCoordinate() {vTiling=0}
  n43 = ComponentMask(None=VectorParameter() {parameterName=MaskBottomOffset, defaultValue=<rgba 0.5 0 0 0>}.RGB) {g=true, r=true}
  n44 = ScalarParameter() {defaultValue=0.25, parameterName=MaskBottom_Radius}
  n45 = ScalarParameter() {parameterName=MaskBottom_Hardness}
  n46 = Power(Base=OneMinus(None=SphereMask(A=@n42, B=@n43, Radius=@n44, Hardness=@n45)))
  n47 = TextureCoordinate()
  n48 = Multiply(
   A=Multiply(A=Multiply(A=@n25.A, B=@n40), B=LinearInterpolate(A=@n41, Alpha=@n46))
   B=Power(Base=ComponentMask(None=@n47) {g=true}, Exp=ScalarParameter() {defaultValue=4, parameterName=FireWidth})
  )
  n49 = TextureSampleParameter2D() {group=AlphaNoise, parameterName=AlphaNoise, samplerType=SAMPLERTYPE_Masks, texture=/Game/Weapons/WeaponSkins/Textures/GunAnimationGradient}
  n50 = Multiply(A=@n48, B=@n49.R)
  n51 = ScalarParameter() {defaultValue=2, group=AlphaNoise, parameterName=AlphaNoise2_TileX}
  n52 = ScalarParameter() {defaultValue=1, group=AlphaNoise, parameterName=AlphaNoise2_TileY}
  n53 = TextureSampleParameter2D(UVs=Multiply(A=TextureCoordinate(), B=AppendVector(A=@n51, B=@n52))) {group=AlphaNoise, parameterName=AlphaNoise2, texture=/Game/Weapons/FORT_Rifles/Effects/Textures/T_FX_SoundBeam_01}
  StaticSwitchParameter(True=StaticSwitchParameter(True=Multiply(A=@n50, B=@n53.RGB), False=@n50) {group=AlphaNoise, parameterName=UseAlphaNoise2}, False=@n48) {group=AlphaNoise, parameterName=UseAlphaNoise}
 ]
}
