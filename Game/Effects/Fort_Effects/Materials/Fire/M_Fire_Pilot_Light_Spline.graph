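# Translucent, unlit, two-sided material for a blue pilot-light flame drawn on spline meshes widened by SplineThicken.
# World position offset adds panning swirl-noise wiggle and Z-axis height growth, masked by vertex red; vertex green varies speed, noise phase and height.
# Panning noise distorts the flame texture's UVs and modulates the HDR two-color emissive.
# Vertex green and blue form a mask driving RadialGrow and a rounded, hard-edged Reveal on opacity.
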Material /Game/Effects/Fort_Effects/Materials/Fire/M_Fire_Pilot_Light_Spline
graph {
 properties {blendMode=BLEND_Translucent, pixelDepthOffsetMode=PDOM_Legacy, shadingModel=MSM_Unlit, twoSided=true, usedShadingModels=Unlit}
 nodes [
  n1 = ScalarParameter() {defaultValue=8, parameterName=WidthBase}
  n2 = ScalarParameter() {defaultValue=7, parameterName=WidthTip}
  n3 = VertexColor()
  n4 = Multiply(A=Time(), B=Multiply(A=ScalarParameter() {defaultValue=40, parameterName=Speed}, B=LinearInterpolate(Alpha=@n3.G) {constA=0.25}))
  n5 = Panner(Coordinate=Add(A=TextureCoordinate() {uTiling=2, vTiling=2}, B=VertexColor().G), Time=@n4) {speedX=0.05, speedY=0.1}
  n6 = Transform(None=Multiply(A=TextureSample(UVs=@n5) {samplerType=SAMPLERTYPE_Normal, texture=/Game/Global/Textures/Default/Noise/T_SwirlNoise_N}.RGB, B=VertexColor().R)) {transformSourceType=TRANSFORMSOURCE_Local}
  n7 = ScalarParameter() {defaultValue=2, parameterName=WiggleAmount}
  n8 = ComponentMask(None=Multiply(A=WorldPosition().XYZ) {constB=0.0005}) {g=true, r=true}
  n9 = Multiply(
   A=TextureSample(UVs=Panner(Coordinate=@n8, Time=@n4) {speedX=0.1, speedY=0.2}) {samplerType=SAMPLERTYPE_Normal, texture=&T_SwirlNoise_N}.R
   B=ScalarParameter() {defaultValue=8, parameterName=HeightWiggle}
  )
  n10 = ScalarParameter() {defaultValue=-9, parameterName=Grow}
  n11 = VertexColor()
  n12 = Add(A=@n11.G) {constB=0.25}
  n13 = OneMinus(None=Multiply(A=Clamp(None=@n12), B=@n11.B))
  n14 = Add(A=@n13, B=ScalarParameter() {defaultValue=1, parameterName=RadialGrow})
  n15 = Add(A=Add(A=Add(A=@n9, B=@n10), B=Multiply(A=@n11.G) {constB=15}), B=Multiply(A=Clamp(None=@n14)) {constB=5})
  n16 = Transform(None=Constant3Vector() {constant=<rgba 0 0 1 0>}) {transformSourceType=TRANSFORMSOURCE_Local}
  Add(
   A=Add(
    A=MaterialFunctionCall(WidthBase=@n1, WidthTip=@n2) {materialFunction=&SplineThicken}.WorldPositionOffset
    B=Multiply(A=@n6, B=@n7)
   )
   B=Multiply(A=Multiply(A=@n15, B=@n16), B=VertexColor().R)
  )
  n17 = TextureCoordinate() {vTiling=0.5}
  n18 = VertexColor()
  n19 = TextureSample(UVs=Panner(Coordinate=Add(A=@n17, B=@n18.G), Time=@n4) {speedY=2}) {samplerType=SAMPLERTYPE_Normal, texture=&T_SwirlNoise_N}
  n20 = ComponentMask(None=@n19.RGB) {g=true, r=true}
  n21 = ScalarParameter() {defaultValue=0.025, parameterName=NoiseScale}
  n22 = TextureSample(UVs=Add(A=TextureCoordinate(), B=Multiply(A=@n20, B=@n21))) {texture=/Game/Effects/Fort_Effects/Textures/Fire/T_Fire_PilotLight}
  n23 = ScalarParameter() {defaultValue=0.15, parameterName=Opacity}
  n24 = Add(A=@n13, B=ScalarParameter() {defaultValue=-0.5, parameterName=Reveal})
  Multiply(
   A=Multiply(A=Multiply(A=@n22.G, B=@n23), B=ScalarParameter() {defaultValue=0.35, parameterName=OverallOpacity})
   B=Round(None=Clamp(None=@n24))
  )
  n25 = LinearInterpolate(
   A=VectorParameter() {parameterName=ColorA, defaultValue=<rgba 0.2 1.6232 4 1>}.RGB
   B=VectorParameter() {parameterName=ColorB, defaultValue=<rgba 1.0376 1.9403 6 1>}.RGB
   Alpha=@n22.G
  )
  Multiply(
   A=MaterialFunctionCall(LDR Input=Clamp(None=@n25), HDR Tint=@n25) {materialFunction=&DeriveHDRfromLDR}.RemappedRange
   B=LinearInterpolate(Alpha=@n19.R) {constA=0.5}
  )
 ]
}
