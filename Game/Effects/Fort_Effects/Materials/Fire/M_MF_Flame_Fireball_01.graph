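# Unlit additive particle-sprite material for a fireball flame effect, built on a sub-UV flipbook texture.
# The texture's red channel is tinted by particle color, brightened, and passed through the MF_FX_EmissiveAdjustments function.
# The same channel times particle alpha feeds a depth fade, softening intersections with geometry.
# A Desaturate static switch reads the texture RGB, but its output feeds no other node shown.
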Material /Game/Effects/Fort_Effects/Materials/Fire/M_MF_Flame_Fireball_01
graph {
 properties {bUsedWithParticleSprites=true, blendMode=BLEND_Additive, pixelDepthOffsetMode=PDOM_Legacy, shadingModel=MSM_Unlit, usedShadingModels=Unlit}
 nodes [
  n1 = ParticleSubUV!(UVs, Tex, Apply View MipBias) {texture=/Game/Effects/Fort_Effects/Textures/T_Fire_Ball_Sub_UV3}
  n2 = ParticleColor()
  DepthFade(Opacity=Multiply(A=@n1.R, B=@n2.A)) {fadeDistanceDefault=30}
  MaterialFunctionCall(Emissive=Multiply(A=Multiply(A=@n2.RGB, B=@n1.R)) {constB=4}) {materialFunction=&MF_FX_EmissiveAdjustments}
  StaticSwitchParameter(True=Desaturation(None=@n1.RGB), False=@n1.RGB) {parameterName=Desaturate}
 ]
}
