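# Masked, dithered Default Lit sprite material for fire-effect ash particles, with a flame-noise normal map and per-particle random UV offsets.
# Tiling noise, a radial gradient and clamped camera depth fade form a shape that particle alpha erodes into a hard cutout.
# A vertex-shader offset pulls sprites toward their particle center near the camera.
# A static switch routes particle color to either emissive adjustments or a base-color tint.
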Material /Game/Effects/Fort_Effects/Materials/Fire/M_MaskedDeath_Ash
graph {
 properties {
  bUsedWithNiagaraSprites=true
  bUsedWithParticleSprites=true
  blendMode=BLEND_Masked
  ditherOpacityMask=true
  pixelDepthOffsetMode=PDOM_Legacy
  translucencyLightingMode=TLM_VolumetricPerVertexDirectional
  twoSided=true
  usedShadingModels=Default Lit
 }
 nodes [
  n1 = ParticleColor()
  n2 = Add(A=Multiply(A=ParticleRandom(), B=Constant2Vector() {g=68.1, r=6.7}), B=Multiply(A=TextureCoordinate()) {constB=0.33})
  n3 = TextureSample(UVs=@n2) {samplerType=SAMPLERTYPE_Grayscale, texture=/Game/Global/Textures/Default/Noise/TilingNoise06}
  n4 = ScalarParameter() {defaultValue=2000, parameterName=Camera Depth Fade Distance}
  n5 = Multiply(
   A=Multiply(
    A=@n3.G
    B=MaterialFunctionCall() {materialFunction=&RadialGradientExponential}.RadialGradientExponential
   )
   B=Clamp(
    None=MaterialFunctionCall(Fade Length=@n4) {materialFunction=&CameraDepthFade}.Result
    Min=ScalarParameter() {defaultValue=0.6, parameterName=Min Cam Depth Scale}
   )
  )
  n6 = OneMinus(None=ConstantBiasScale(None=@n1.A) {scale=0.4})
  Ceil(None=Multiply(A=@n1.A, B=Subtract(A=@n5, B=@n6)))
  Constant()
  LinearInterpolate(Alpha=@n3.G) {constA=1, constB=0}
  n7 = WorldPosition()
  n8 = ParticlePositionWS()
  LinearInterpolate(
   A=Multiply(A=Subtract(A=@n7.XYZ, B=@n8)) {constB=-1}
   Alpha=MaterialFunctionCall(Fade Length=Constant() {r=170}, For Vertex Shader=StaticBool() {value=true}) {materialFunction=&CameraDepthFade}.Result
  ) {constB=0}
  n9 = ParticleColor()
  MaterialFunctionCall(Emissive=StaticSwitchParameter(True=@n9.RGB, False=Constant()) {parameterName=Use Emissive Decay}) {materialFunction=&MF_FX_EmissiveAdjustments}
  Reroute(None=TextureSample(UVs=@n2) {samplerType=SAMPLERTYPE_Normal, texture=/Game/Effects/Fort_Effects/Textures/Tile/noise_flamy_Tile_NRM}.RGB)
  n10 = VectorParameter() {parameterName=BaseColor, defaultValue=<rgba 0.2 0.2 0.2 0>}
  StaticSwitchParameter(True=@n10.RGB, False=Multiply(A=@n10.RGB, B=@n9.RGB)) {parameterName=Use Emissive Decay}
 ]
}
