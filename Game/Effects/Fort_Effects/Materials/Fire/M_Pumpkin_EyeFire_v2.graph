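# Translucent, two-sided unlit material for animated pumpkin eye fire on mesh particles or skeletal meshes, green-teal by default.
# World position offset combines spline thickening, panning waves and optional local offsets weighted along V.
# Panning noise drives HDR color and, distorted and shaped by procedural edge, split and bottom masks, floor-stepped opacity.
# Depth fade, optional view-angle fade and particle alpha finish opacity; mobile skips distortion and both fades.
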Material /Game/Effects/Fort_Effects/Materials/Fire/M_Pumpkin_EyeFire_v2
graph {
 properties {
  bUsedWithMeshDeformer=true
  bUsedWithMeshParticles=true
  bUsedWithNiagaraMeshParticles=true
  bUsedWithSkeletalMesh=true
  blendMode=BLEND_Translucent
  pixelDepthOffsetMode=PDOM_Legacy
  shadingModel=MSM_Unlit
  twoSided=true
  usedShadingModels=Unlit
 }
 nodes [
  n1 = ScalarParameter() {defaultValue=13, group=WPO, parameterName=WidthTip, sortPriority=1}
  n2 = ScalarParameter() {defaultValue=7, group=WPO, parameterName=WidthBase, sortPriority=2}
  n3 = Panner(
   Coordinate=LinearInterpolate(A=TextureCoordinate() {coordinateIndex=1}, B=TextureCoordinate(), Alpha=ScalarParameter() {parameterName=Use Defaut UV type})
   Time=Multiply(A=Time(), B=ScalarParameter() {defaultValue=1, group=Panner Speed, parameterName=Speed})
  ) {speedY=0.8}
  n4 = AppendVector(
   A=ComponentMask(None=TextureSample(UVs=@n3) {samplerType=SAMPLERTYPE_Normal, texture=&T_Weap_Distort_N_02}.RGB) {g=true, r=true}
   B=TextureSample(UVs=@n3) {texture=&T_Smooth_tile}.R
  )
  n5 = OneMinus(None=ComponentMask(None=TextureCoordinate()) {g=true})
  n6 = Add(
   A=LinearInterpolate(
    B=MaterialFunctionCall(WidthBase=@n1, WidthTip=@n2) {materialFunction=&SplineThicken}.WorldPositionOffset
    Alpha=ScalarParameter() {defaultValue=1, group=WPO, parameterName=Spline Thicken}
   )
   B=Multiply(A=LinearInterpolate(B=@n4, Alpha=@n5), B=ScalarParameter() {defaultValue=10, group=WPO, parameterName=WaveIntensity, sortPriority=7}) {constB=10}
  )
  n7 = ComponentMask(None=TextureCoordinate()) {g=true}
  n8 = Multiply(
   A=Transform(None=VectorParameter() {group=WPO, parameterName=LocalPostioning, sortPriority=4}.RGB) {transformSourceType=TRANSFORMSOURCE_Local}
   B=Power(Base=OneMinus(None=@n7), Exp=ScalarParameter() {defaultValue=2, group=WPO, parameterName=Edge Gradient Movement Control Exp, sortPriority=6})
  )
  n9 = ComponentMask(None=TextureCoordinate()) {g=true}
  n10 = Multiply(
   A=Transform(None=VectorParameter() {group=WPO, parameterName=Initial WPO Offset, sortPriority=5}.RGB) {transformSourceType=TRANSFORMSOURCE_Local}
   B=Power(Base=OneMinus(None=@n9), Exp=ScalarParameter() {defaultValue=2, group=WPO, parameterName=Edge Gradient Movement Control Exp, sortPriority=6})
  )
  LinearInterpolate(A=@n6, B=Add(A=@n6, B=Add(A=@n8, B=@n10)), Alpha=ScalarParameter() {group=WPO, parameterName=Add Local Position Control, sortPriority=3})
  n11 = TextureCoordinate() {coordinateIndex=1, uTiling=0.3, vTiling=0.7}
  n12 = TextureCoordinate() {uTiling=0.3, vTiling=0.7}
  n13 = ScalarParameter() {parameterName=Use Defaut UV type}
  n14 = VectorParameter() {group=Color, parameterName=Fire Color UVScale, sortPriority=2, defaultValue=<rgba 1 1 0 0>}
  n15 = Panner(
   Coordinate=Multiply(A=LinearInterpolate(A=@n11, B=@n12, Alpha=@n13), B=ComponentMask(None=@n14.RGB) {g=true, r=true})
   Time=Multiply(A=Time(), B=ScalarParameter() {defaultValue=1, group=Panner Speed, parameterName=Speed})
  ) {speedY=0.6}
  n16 = LinearInterpolate(
   A=VectorParameter() {group=Color, parameterName=ColorLow, sortPriority=3, defaultValue=<rgba 0 0.5052 0.387 1>}.RGB
   B=VectorParameter() {group=Color, parameterName=ColorHigh, sortPriority=4, defaultValue=<rgba 0 1 0.493 1>}.RGB
   Alpha=ChannelMaskParameter(None=TextureSampleParameter2D(UVs=@n15) {group=Color, parameterName=Fire Color Lerp Texture, texture=/Game/Global/Textures/Default/Noise/T_FireNoise_02}.RGBA) {group=Color, parameterName=Color Lerp Mask, sortPriority=1}.R
  )
  n17 = ScalarParameter() {defaultValue=1, group=Color, parameterName=HDR Intensity Multiplier, sortPriority=7}
  n18 = LinearInterpolate(
   A=TextureCoordinate() {coordinateIndex=1, vTiling=1.5}
   B=TextureCoordinate() {vTiling=1.5}
   Alpha=ScalarParameter() {parameterName=Use Defaut UV type}
  )
  n19 = ComponentMask(None=VectorParameter() {group=Distortion, parameterName=Distortion UV Scale, sortPriority=3, defaultValue=<rgba 1 1 0 0>}.RGB) {g=true, r=true}
  n20 = Time()
  n21 = ScalarParameter() {defaultValue=1, group=Panner Speed, parameterName=Speed}
  n22 = TextureSampleParameter2D(UVs=Panner(Coordinate=Multiply(A=@n18, B=@n19), Time=Multiply(A=@n20, B=@n21)) {speedY=1}) {
   group=Distortion
   parameterName=Distortion Normal
   samplerType=SAMPLERTYPE_Normal
   sortPriority=2
   texture=&T_Weap_Distort_N_02
  }
  n23 = ComponentMask(None=MaterialFunctionCall(Normal=@n22.RGB, Flatness=ScalarParameter() {group=Distortion, parameterName=UV_Normal_Flatness, sortPriority=4}) {materialFunction=&FlattenNormal}.Result) {g=true, r=true}
  n24 = TextureCoordinate()
  n25 = VectorParameter() {group=Fire Shape, parameterName=Fire_MacroShape_UVScale, sortPriority=5, defaultValue=<rgba 1 1 0 0>}
  n26 = ComponentMask(None=@n25.RGB) {g=true, r=true}
  n27 = ComponentMask(None=Multiply(A=@n24, B=@n26)) {g=true}
  n28 = LinearInterpolate(
   B=Divide(
    A=LinearInterpolate(A=@n23, Alpha=@n27) {constB=0}
    B=ScalarParameter() {defaultValue=3, group=Distortion, parameterName=UV_Distortion_Divsor, sortPriority=1}
   ) {constB=20}
   Alpha=ScalarParameter() {defaultValue=1, group=Distortion, parameterName=UV_Distortion_Lerp}
  )
  n29 = LinearInterpolate(
   A=TextureCoordinate() {coordinateIndex=1, uTiling=0.8, vTiling=0.9}
   B=TextureCoordinate() {uTiling=0.8, vTiling=0.9}
   Alpha=ScalarParameter() {parameterName=Use Defaut UV type}
  )
  n30 = ComponentMask(None=VectorParameter() {group=Fire Shape, parameterName=Fire_Micro_UVScale, sortPriority=8, defaultValue=<rgba 1 1 0 0>}.RGB) {g=true, r=true}
  n31 = Panner(
   Coordinate=MaterialFunctionCall(XY=Multiply(A=@n29, B=@n30)) {materialFunction=&Swizzle}.YX
   Time=Multiply(A=Time(), B=ScalarParameter() {defaultValue=1, group=Panner Speed, parameterName=Speed})
  ) {speedX=1.2, speedY=0.2}
  n32 = TextureSampleParameter2D(UVs=ShadingPathSwitch(Default=Add(A=@n28, B=@n31), Mobile=@n31) {desc=Migrated from FeatureLevelSwitch}) {
   group=Fire Shape
   parameterName=Fire Micro Texture
   samplerType=SAMPLERTYPE_LinearColor
   sortPriority=6
   texture=&T_NoisePack_LinearColor
  }
  n33 = ChannelMaskParameter(None=@n32.RGBA) {group=Fire Shape, parameterName=Fire Micro Mask, sortPriority=7}
  n34 = LinearInterpolate(A=@n17, B=ScalarParameter() {defaultValue=5, group=Color, parameterName=HDR Intensity Multiplier Pulse High, sortPriority=9}, Alpha=@n33.R) {constA=1, constB=5}
  n35 = ScalarParameter() {group=Color, parameterName=Pulse HDR Intensity, sortPriority=8}
  n36 = ScalarParameter() {defaultValue=5, group=Color, parameterName=EmissiveAmount, sortPriority=6}
  MaterialFunctionCall(Emissive=MaterialFunctionCall(
    LDR Input=@n16
    HDR Dynamic Range=ScalarParameter() {defaultValue=5, group=Color, parameterName=HDR Dynamic Range, sortPriority=5}
    HDR Tint=Multiply(A=@n16, B=@n36) {constA=5}
    HDR Intensity Multiplier=LinearInterpolate(A=@n17, B=@n34, Alpha=@n35)
   ) {materialFunction=&DeriveHDRfromLDR}.RemappedRange) {materialFunction=&MF_FX_EmissiveAdjustments}
  n37 = MaterialFunctionCall(Float2=Multiply(A=@n24, B=@n26)) {materialFunction=&BreakOutFloat2Components}
  n38 = Power(
   Base=Sine(None=Multiply(A=@n37.R) {constB=0.5})
   Exp=LinearInterpolate(
    A=ScalarParameter() {defaultValue=6, group=Fire Shape, parameterName=Edge Fade Value Top, sortPriority=3}
    B=ScalarParameter() {defaultValue=2, group=Fire Shape, parameterName=Edge Fade Value Bottom, sortPriority=4}
    Alpha=@n37.G
   ) {constA=6, constB=2}
  ) {constExponent=4}
  n39 = VectorParameter() {group=BottomFade, parameterName=Bottom Fade_SphereMask_A_Scale, sortPriority=2, defaultValue=<rgba 1.2 1 0 0>}
  n40 = Multiply(A=TextureCoordinate(), B=ComponentMask(None=@n39.RGB) {g=true, r=true})
  n41 = VectorParameter() {group=BottomFade, parameterName=Bottom Fade_SphereMask_B, sortPriority=3, defaultValue=<rgba 0.6 0.05 0 0>}
  n42 = ComponentMask(None=@n41.RGB) {g=true, r=true}
  n43 = ScalarParameter() {defaultValue=0.95, group=BottomFade, parameterName=BottomFade_Radius, sortPriority=1}
  n44 = Saturate(None=Multiply(A=SphereMask(A=@n40, B=@n42, Radius=@n43) {attenuationRadius=0.95, hardnessPercent=0}) {constB=15})
  n45 = ScalarParameter() {defaultValue=1, group=BottomFade, parameterName=Bottom Fade}
  n46 = Multiply(A=@n33.R, B=Multiply(A=Multiply(A=@n38, B=@n37.G), B=LinearInterpolate(B=@n44, Alpha=@n45) {constA=1}))
  n47 = LinearInterpolate(Alpha=@n37.G) {constA=2}
  n48 = ScalarParameter() {defaultValue=2, group=Fire Shape, parameterName=Fire Split Contrast, sortPriority=1}
  n49 = OneMinus(None=Sine(None=Multiply(A=@n37.R) {constB=0.45}))
  n50 = LinearInterpolate(Alpha=@n22.R) {constA=0.5, constB=1.5}
  n51 = Round(None=Add(A=Power(Base=@n37.G, Exp=@n48), B=Power(Base=@n49, Exp=@n50) {constExponent=0.5}))
  n52 = ScalarParameter() {defaultValue=1, group=Fire Shape, parameterName=Fire Split}
  n53 = Multiply(
   A=Multiply(A=Power(Base=@n46, Exp=@n47), B=LinearInterpolate(B=@n51, Alpha=@n52) {constA=1, constB=0})
   B=ScalarParameter() {defaultValue=2, group=Opacity, parameterName=OpacityMaskMultiplier, sortPriority=1}
  ) {constB=3}
  n54 = LinearInterpolate(A=Floor(None=@n53), B=@n53, Alpha=ScalarParameter() {parameterName=Floor})
  n55 = DepthFade(Opacity=@n54, FadeDistance=ScalarParameter() {defaultValue=24, group=Opacity, parameterName=DepthFadeDistance})
  n56 = LinearInterpolate(
   B=MaterialFunctionCall("FadeSensitivity (optional)"=ScalarParameter() {parameterName=Fade Sensitivity}) {materialFunction=&FadeBasedOnViewAngle}.Result
   Alpha=ScalarParameter() {parameterName=Fade based on View Angle}
  ) {constA=1}
  Multiply(
   A=ShadingPathSwitch(Default=Multiply(A=@n55, B=@n56), Mobile=@n54) {desc=Migrated from FeatureLevelSwitch}
   B=LinearInterpolate(B=ParticleColor().A, Alpha=ScalarParameter() {defaultValue=1, parameterName=Multiply by Particle Alpha}) {constA=1}
  )
 ]
}
